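# Animation Blueprint for the Snowman backpack cosmetic that adds procedural secondary motion to its neck, arms and scarf.
# The AnimGraph passes an input pose through chained Anim Dynamics nodes with gravity overrides and angular limits; the two scarf bones add damped angular springs.
# Each update, the EventGraph multiplies targeting and vehicle factors into one shared alpha, reducing the simulation while aiming and disabling it in vehicles.
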AnimBlueprint /BRCosmetics/Accessories/FORT_Backpacks/Backpack_Snowman/Meshes/M_MED_Snowman_AnimBlueprint
target_skeleton=&/M_MED_Snowman_Skeleton
graphs [
 AnimGraph {nodes [
   AnimGraphNode_SubInput_0 = Input Pose()
   n1 = Get BackPack Alpha() -> (BackPack Alpha:double)
   n2 = Anim Dynamics - Bone: dyn_neck(ComponentPose=Local To Component(LocalPose=@AnimGraphNode_SubInput_0), Alpha=@n1, ExternalForce="0,0,0") {
    bGravityOverrideInSimSpace=true
    bUseGravityOverride=true
    alphaBoolBlend {blendOption=Linear}
    boundBone {boneName=dyn_neck}
    componentLinearAccScale=<1 1 1>
    componentLinearVelScale=<1 1 1>
    gravityOverride {z=-980}
    physicsBodyDefinitions [{
      sphereCollisionRadius=0
      boundBone {boneName=dyn_neck}
      boxExtents=<10 10 5>
      constraintSetup {angularTargetAxis=AxisY, twistAxis=AxisZ, angularLimitsMax=<5 15 0>, angularLimitsMin=<-10 -15 0>, angularTarget=<0 1 0>}
      localJointOffset=<0 -0.01 0.1>
     }]
   }
   n3 = Anim Dynamics - Bone: dyn_left_arm(ComponentPose=@n2, Alpha=@n1, ExternalForce="0,0,0") {
    bGravityOverrideInSimSpace=true
    bUseGravityOverride=true
    alphaBoolBlend {blendOption=Linear}
    boundBone {boneName=dyn_left_arm}
    componentLinearAccScale=<1 1 1>
    componentLinearVelScale=<1 1 1>
    gravityOverride {z=-980}
    physicsBodyDefinitions [{
      sphereCollisionRadius=0
      boundBone {boneName=dyn_left_arm}
      boxExtents=<15 3 3>
      constraintSetup {angularLimitsMax=<0 45 10>, angularLimitsMin=<0 -20 -10>, angularTarget=<0 0 0>}
      localJointOffset=<-1 0 0>
     }]
   }
   n4 = Anim Dynamics - Bone: dyn_right_arm(ComponentPose=@n3, Alpha=@n1, ExternalForce="0,0,0") {
    bGravityOverrideInSimSpace=true
    bUseGravityOverride=true
    alphaBoolBlend {blendOption=Linear}
    boundBone {boneName=dyn_right_arm}
    componentLinearAccScale=<1 1 1>
    componentLinearVelScale=<1 1 1>
    gravityOverride {z=-980}
    physicsBodyDefinitions [{
      sphereCollisionRadius=0
      boundBone {boneName=dyn_right_arm}
      boxExtents=<-15 3 3>
      constraintSetup {angularLimitsMax=<0 20 10>, angularLimitsMin=<0 -45 -10>, angularTarget=<0 0 0>}
      localJointOffset=<1 0 0>
     }]
   }
   n5 = Anim Dynamics - Bone: dyn_scarf_01(ComponentPose=@n4, Alpha=@n1, ExternalForce="0,0,0") {
    angularDampingOverride=0.982
    angularSpringConstant=200
    bAngularSpring=true
    bGravityOverrideInSimSpace=true
    bOverrideAngularDamping=true
    bUseGravityOverride=true
    alphaBoolBlend {blendOption=Linear}
    boundBone {boneName=dyn_scarf_01}
    componentLinearAccScale=<1 1 1>
    componentLinearVelScale=<1 1 1>
    gravityOverride {z=-980}
    physicsBodyDefinitions [{
      sphereCollisionRadius=0
      boundBone {boneName=dyn_scarf_01}
      boxExtents=<3 5 5>
      constraintSetup {angularLimitsMax=<0 15 25>, angularLimitsMin=<0 -20 -25>, angularTarget=<1 0 0>}
      localJointOffset=<0.2 0 0.01>
     }]
   }
   n6 = Anim Dynamics - Bone: dyn_scarf_02(ComponentPose=@n5, Alpha=@n1, ExternalForce="0,0,0") {
    angularDampingOverride=0.982
    angularSpringConstant=200
    bAngularSpring=true
    bGravityOverrideInSimSpace=true
    bOverrideAngularDamping=true
    bUseGravityOverride=true
    alphaBoolBlend {blendOption=Linear}
    boundBone {boneName=dyn_scarf_02}
    componentLinearAccScale=<1 1 1>
    componentLinearVelScale=<1 1 1>
    gravityOverride {z=-980}
    physicsBodyDefinitions [{
      sphereCollisionRadius=0
      boundBone {boneName=dyn_scarf_02}
      boxExtents=<3 5 5>
      constraintSetup {angularTargetAxis=AxisZ, angularLimitsMax=<0 15 25>, angularLimitsMin=<0 -20 -25>, angularTarget=<0 0 1>}
      localJointOffset=<0.2 0 0.01>
     }]
   }
   Output Pose(Result=Component To Local(ComponentPose=@n6)) {layerGroup=None}
  ]}
 EventGraph {nodes [
   Event BlueprintUpdateAnimation()
   > n1 = Sequence!(execute) -> (then_0, then_1, then_2:Exec)
   > n2 = Branch(Condition=Get bIsTargeting())
   > Set target_alpha(target_alpha:double=0.2) -> (Output_Get:double)
   Set target_alpha(execute=@n2.else, target_alpha=1)
   n3 = Branch(execute=@n1.then_1, Condition=Get bIsInVehicle() -> (bIsInVehicle:bool))
   > Set Is In Vehicle Alpha(Is In Vehicle Alpha:double=0) -> (Output_Get:double)
   Set Is In Vehicle Alpha(execute=@n3.else, Is In Vehicle Alpha=1)
   Set BackPack Alpha(
    execute=@n1.then_2
    BackPack Alpha:double=float * float!(A=Get target_alpha() -> (target_alpha:double), B=Get Is In Vehicle Alpha() -> (Is In Vehicle Alpha:double), C:double=1) -> (ReturnValue) {is=title K2Node_PromotableOperator}
   ) -> (Output_Get:double)
   TryGetPawnOwner()
  ]}
]
variables [
 BackPack Alpha {type=number}
 Is in Vehicle Alpha {type=number, default=0.2}
 target_alpha {type=number, default=1}
]
